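# Sound cue for the destruction of a basic tier-two blunt sledgehammer melee weapon.
# A mixer layers an axe destroy sample with a sledgehammer hit that a modulator drops to a fixed half pitch for weight.
# Playback is quiet, using the small melee weapon attenuation asset, the melee weapon sound class and an overridden concurrency rule that prevents new instances at the limit.
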SoundCue /Game/Sounds/Fort_Weapons/SledgeHammers/Melee_Blunt_T02_Basic_Sledgehammer_Destroyed_Cue
root=Mixer_2
nodes [
 {id=Mixer_2, children [Modulator_1,WavePlayer_4], properties {inputVolume [1,1]}}
 {id=Modulator_1, children [WavePlayer_3], properties {pitchMax=0.5, pitchMin=0.5}}
 {id=WavePlayer_3, properties {soundWaveAssetPtr=&/sledgehammer_hit_ST_02}}
 {id=WavePlayer_4, properties {soundWaveAssetPtr=/Game/Sounds/Fort_Weapons/Axes/axe_destroy_02}}
]
properties {
 attenuationSettings=&Fort_Weapon_Melee_Sml_Att
 bOverrideConcurrency=true
 maxDistance=2800
 soundClassObject=&Fort_WeaponMelee_PSM
 volumeMultiplier=0.3
 attenuationOverrides {distanceAlgorithm=NaturalSound, falloffDistance=3000, attenuationShapeExtents {x=1000}}
 concurrencyOverrides {resolutionRule=PreventNew}
}
