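# Sound cue for the impact of a basic tier 2 sledgehammer melee weapon.
# A quality-level switch fills two branches, the second with fewer random sledge variants, and leaves two empty.
# Each branch mixes a hit layer with a quieter random sledge layer, each with its own sound class and attenuation.
# The hit layer gets variety from one wave, randomly played directly or through pitch envelopes, one also modulated.
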SoundCue /Game/Sounds/Fort_Weapons/SledgeHammers/Melee_Blunt_T02_Basic_Sledgehammer_Impact_Cue
root=QualityLevel_0
nodes [
 {id=Attenuation_2, children [Random_2], properties {attenuationSettings=&Fort_Weapon_Melee_Sml_Att}}
 {
  id=Attenuation_3
  children [Random_3]
  properties {attenuationSettings=/Game/Sounds/Attenuation/Fort_Weapon_Melee_Attack_Close_Range_Stereo_Att}
 }
 {id=Attenuation_4, children [Random_4], properties {attenuationSettings=&Fort_Weapon_Melee_Sml_Att}}
 {
  id=Attenuation_5
  children [Random_5]
  properties {attenuationSettings=&Fort_Weapon_Melee_Attack_Close_Range_Stereo_Att}
 }
 {
  id=Enveloper_2
  children [Modulator_1]
  properties {pitchCurve {editorCurveData {keys [{time=0}, <at 0.2 1.07>, {time=1}]}}, volumeCurve {editorCurveData {keys [<at 0 1>]}}}
 }
 {
  id=Enveloper_3
  children [WavePlayer_6]
  properties {pitchCurve {editorCurveData {keys [<at 0 1.1>, {time=0.3}, <at 0.3 1.07>]}}, volumeCurve {editorCurveData {keys [<at 0 1>]}}}
 }
 {
  id=Enveloper_4
  children [Modulator_2]
  properties {pitchCurve {editorCurveData {keys [{time=0}, <at 0.2 1.07>, {time=1}]}}, volumeCurve {editorCurveData {keys [<at 0 1>]}}}
 }
 {
  id=Enveloper_5
  children [WavePlayer_12]
  properties {pitchCurve {editorCurveData {keys [<at 0 1.1>, {time=0.3}, <at 0.3 1.07>]}}, volumeCurve {editorCurveData {keys [<at 0 1>]}}}
 }
 {id=Mixer_1, children [SoundClass_2,SoundClass_3], properties {inputVolume [1,0.7]}}
 {id=Mixer_2, children [SoundClass_4,SoundClass_5], properties {inputVolume [1,0.7]}}
 {id=Modulator_1, children [WavePlayer_6], properties {pitchMin=1}}
 {id=Modulator_2, children [WavePlayer_12], properties {pitchMin=1}}
 {id=QualityLevel_0, children [Mixer_1,Mixer_2,null,null]}
 {id=Random_2, children [WavePlayer_6,Enveloper_2,Enveloper_3], properties {bRandomizeWithoutReplacement=false, weights [1,1,1]}}
 {id=Random_3, children [WavePlayer_7, WavePlayer_8, WavePlayer_9, WavePlayer_10, WavePlayer_11], properties {weights [1,1,1,1,1]}}
 {id=Random_4, children [WavePlayer_12, Enveloper_4, Enveloper_5], properties {bRandomizeWithoutReplacement=false, weights [1,1,1]}}
 {id=Random_5, children [WavePlayer_13, WavePlayer_14, WavePlayer_17], properties {weights [1,1,1]}}
 {id=SoundClass_2, children [Attenuation_2], properties {soundClassOverride=&Fort_WeaponMelee_PSM}}
 {id=SoundClass_3, children [Attenuation_3], properties {soundClassOverride=&Fort_WeaponMelee}}
 {id=SoundClass_4, children [Attenuation_4], properties {soundClassOverride=&Fort_WeaponMelee_PSM}}
 {id=SoundClass_5, children [Attenuation_5], properties {soundClassOverride=&Fort_WeaponMelee}}
 {id=WavePlayer_10, properties {soundWaveAssetPtr=&/sledge_new_04}}
 {id=WavePlayer_11, properties {soundWaveAssetPtr=&/sledge_new_05}}
 {id=WavePlayer_12, properties {soundWaveAssetPtr=&/sledgehammer_hit_ST_01}}
 {id=WavePlayer_13, properties {soundWaveAssetPtr=&/sledge_new_01}}
 {id=WavePlayer_14, properties {soundWaveAssetPtr=&/sledge_new_02}}
 {id=WavePlayer_17, properties {soundWaveAssetPtr=&sledge_new_05}}
 {id=WavePlayer_6, properties {soundWaveAssetPtr=&sledgehammer_hit_ST_01}}
 {id=WavePlayer_7, properties {soundWaveAssetPtr=&sledge_new_01}}
 {id=WavePlayer_8, properties {soundWaveAssetPtr=&sledge_new_02}}
 {id=WavePlayer_9, properties {soundWaveAssetPtr=&/sledge_new_03}}
]
properties {
 attenuationSettings=&Fort_Weapon_Melee_Attack_Close_Range_Stereo_Att
 maxDistance=10058.75
 soundClassObject=&Fort_WeaponMelee_PSM
 volumeMultiplier=0.5
 attenuationOverrides {distanceAlgorithm=NaturalSound, falloffDistance=3000, attenuationShapeExtents {x=1000}}
 concurrencyOverrides {resolutionRule=PreventNew}
 concurrencySet [&Player_Melee_Weapon_Conc]
}
