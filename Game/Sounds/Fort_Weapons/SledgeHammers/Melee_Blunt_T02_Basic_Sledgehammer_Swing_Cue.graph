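# Sound cue for the swing of a Tier 2 basic sledgehammer melee weapon, with small-melee attenuation and player melee weapon concurrency.
# A Random node picks with equal weight between one swing wave played directly and the same wave through an Enveloper whose pitch curve rises, then falls.
# A root Modulator adds slight random pitch variation on top, so a single sample yields varied swings.
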SoundCue /Game/Sounds/Fort_Weapons/SledgeHammers/Melee_Blunt_T02_Basic_Sledgehammer_Swing_Cue
root=Modulator_7
nodes [
 {
  id=Enveloper_3
  children [WavePlayer_0]
  properties {pitchCurve {editorCurveData {keys [{time=0}, <at 0.3841 1.4789>, <at 4.4481 0.8861>]}}, volumeCurve {editorCurveData {keys [<at 0 1>]}}}
 }
 {id=Modulator_7, children [Random_0], properties {pitchMax=1.1, pitchMin=0.9, volumeMax=1, volumeMin=1}}
 {id=Random_0, children [Enveloper_3,WavePlayer_0], properties {weights [1,1]}}
 {id=WavePlayer_0, properties {soundWaveAssetPtr=&/sledgehammer_swing_ST_01}}
]
properties {
 attenuationSettings=&Fort_Weapon_Melee_Sml_Att
 maxDistance=2800
 soundClassObject=&Fort_WeaponMelee_PSM
 volumeMultiplier=0.8
 attenuationOverrides {distanceAlgorithm=NaturalSound, falloffDistance=2000, attenuationShapeExtents {x=500}}
 concurrencySet [&Player_Melee_Weapon_Conc]
}
