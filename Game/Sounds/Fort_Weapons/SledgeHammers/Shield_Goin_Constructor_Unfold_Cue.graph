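# Sound cue for the Constructor shield unfolding, filed with the sledgehammer weapon sounds.
# A quality-level switch feeds the root: the first tier picks randomly between two equally weighted unfold waves, the second plays a single wave, and the remaining two are empty.
# It uses small melee weapon attenuation, the impacts sound class and player melee weapon concurrency, with the debug flag left on.
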SoundCue /Game/Sounds/Fort_Weapons/SledgeHammers/Shield_Goin_Constructor_Unfold_Cue
root=QualityLevel_0
nodes [
 {id=QualityLevel_0, children [Random_0,WavePlayer_9,null,null]}
 {id=Random_0, children [WavePlayer_7,WavePlayer_8], properties {weights [1,1]}}
 {id=WavePlayer_7, properties {soundWaveAssetPtr=&/Mech_Constructor_Shield_Unfold_v1_01}}
 {id=WavePlayer_8, properties {soundWaveAssetPtr=&/Mech_Constructor_Shield_Unfold_v1_02}}
 {id=WavePlayer_9, properties {soundWaveAssetPtr=&Mech_Constructor_Shield_Unfold_v1_01}}
]
properties {
 attenuationSettings=&Fort_Weapon_Melee_Sml_Att
 bDebug=true
 maxDistance=2800
 soundClassObject=&Fort_Impacts
 volumeMultiplier=0.7
 attenuationOverrides {distanceAlgorithm=NaturalSound, falloffDistance=3000, attenuationShapeExtents {x=1000}}
 concurrencySet [&Player_Melee_Weapon_Conc]
}
