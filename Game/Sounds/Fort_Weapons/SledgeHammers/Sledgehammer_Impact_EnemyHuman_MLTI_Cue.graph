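# Sound cue for sledgehammer impacts, named for hits on human enemies, limited by the player melee weapon concurrency set.
# A quality switch fills two of four slots with distance crossfades, the second using fewer random variants.
# Each crossfade blends modulated close impacts into industrial distant impacts with low-pass attenuation, each layer with its own sound class.
# Cue-level attenuation is disabled, leaving falloff to per-layer attenuation nodes.
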SoundCue /Game/Sounds/Fort_Weapons/SledgeHammers/Sledgehammer_Impact_EnemyHuman_MLTI_Cue
root=QualityLevel_0
nodes [
 {id=Attenuation_2, children [Modulator_1], properties {attenuationSettings=&Fort_Weapon_Melee_Sml_Att}}
 {
  id=Attenuation_3
  children [Random_2]
  properties {
   bOverrideAttenuation=true
   attenuationOverrides {
    bAttenuateWithLPF=true
    distanceAlgorithm=Logarithmic
    falloffDistance=8000
    lPFRadiusMax=5000
    lPFRadiusMin=2000
    attenuationShapeExtents {x=5000}
   }
  }
 }
 {id=Attenuation_4, children [Modulator_2], properties {attenuationSettings=&Fort_Weapon_Melee_Sml_Att}}
 {
  id=Attenuation_5
  children [Random_4]
  properties {
   bOverrideAttenuation=true
   attenuationOverrides {
    bAttenuateWithLPF=true
    distanceAlgorithm=Logarithmic
    falloffDistance=8000
    lPFRadiusMax=5000
    lPFRadiusMin=2000
    attenuationShapeExtents {x=5000}
   }
  }
 }
 {
  id=DistanceCrossFade_1
  children [SoundClass_3,SoundClass_2]
  properties {crossFadeInput [
    {fadeInDistanceEnd=300, fadeInDistanceStart=200, fadeOutDistanceEnd=800, fadeOutDistanceStart=400, volume=1}
    {fadeInDistanceEnd=700, fadeInDistanceStart=300, fadeOutDistanceEnd=10000, fadeOutDistanceStart=7000, volume=1}
   ]}
 }
 {
  id=DistanceCrossFade_2
  children [SoundClass_5,SoundClass_4]
  properties {crossFadeInput [
    {fadeInDistanceEnd=300, fadeInDistanceStart=200, fadeOutDistanceEnd=800, fadeOutDistanceStart=400, volume=1}
    {fadeInDistanceEnd=700, fadeInDistanceStart=300, fadeOutDistanceEnd=10000, fadeOutDistanceStart=7000, volume=1}
   ]}
 }
 {id=Modulator_1, children [Random_3]}
 {id=Modulator_2, children [Random_5]}
 {id=QualityLevel_0, children [DistanceCrossFade_1, DistanceCrossFade_2, null, null]}
 {
  id=Random_2
  children [WavePlayer_17, WavePlayer_12, WavePlayer_18, WavePlayer_13, WavePlayer_14, WavePlayer_15, WavePlayer_19, WavePlayer_16]
  properties {weights [1,1,1,1,1,1,1,1]}
 }
 {id=Random_3, children [WavePlayer_20, WavePlayer_21, WavePlayer_22, WavePlayer_23], properties {weights [1,1,1,1]}}
 {id=Random_4, children [WavePlayer_29, WavePlayer_24, WavePlayer_30, WavePlayer_31], properties {weights [1,1,1,1]}}
 {id=Random_5, children [WavePlayer_32, WavePlayer_33, WavePlayer_34], properties {weights [1,1,1]}}
 {
  id=SoundClass_2
  children [Attenuation_3]
  properties {soundClassOverride=&Fort_WeaponRanged_Fire_Distant}
 }
 {id=SoundClass_3, children [Attenuation_2], properties {soundClassOverride=&Fort_WeaponMelee_PSM}}
 {
  id=SoundClass_4
  children [Attenuation_5]
  properties {soundClassOverride=&Fort_WeaponRanged_Fire_Distant}
 }
 {id=SoundClass_5, children [Attenuation_4], properties {soundClassOverride=&Fort_WeaponMelee_PSM}}
 {id=WavePlayer_12, properties {soundWaveAssetPtr=/Game/Sounds/Fort_Ambient_NonLoop/Industrial/Industrial_Distant_Impact03}}
 {id=WavePlayer_13, properties {soundWaveAssetPtr=/Game/Sounds/Fort_Ambient_NonLoop/Industrial/Industrial_Distant_Impact05}}
 {id=WavePlayer_14, properties {soundWaveAssetPtr=/Game/Sounds/Fort_Ambient_NonLoop/Industrial/Industrial_Distant_Impact06}}
 {id=WavePlayer_15, properties {soundWaveAssetPtr=/Game/Sounds/Fort_Ambient_NonLoop/Industrial/Industrial_Distant_Impact09}}
 {id=WavePlayer_16, properties {soundWaveAssetPtr=/Game/Sounds/Fort_Ambient_NonLoop/Industrial/Industrial_Distant_Impact11}}
 {id=WavePlayer_17, properties {soundWaveAssetPtr=/Game/Sounds/Fort_Ambient_NonLoop/Industrial/Industrial_Distant_Impact02}}
 {id=WavePlayer_18, properties {soundWaveAssetPtr=/Game/Sounds/Fort_Ambient_NonLoop/Industrial/Industrial_Distant_Impact04}}
 {id=WavePlayer_19, properties {soundWaveAssetPtr=/Game/Sounds/Fort_Ambient_NonLoop/Industrial/Industrial_Distant_Impact10}}
 {id=WavePlayer_20, properties {soundWaveAssetPtr=/Game/Sounds/Fort_Weapons/Sledgehammer/sledgehammer_impact_01}}
 {id=WavePlayer_21, properties {soundWaveAssetPtr=/Game/Sounds/Fort_Weapons/Sledgehammer/sledgehammer_impact_02}}
 {id=WavePlayer_22, properties {soundWaveAssetPtr=/Game/Sounds/Fort_Weapons/Sledgehammer/sledgehammer_impact_04}}
 {id=WavePlayer_23, properties {soundWaveAssetPtr=/Game/Sounds/Fort_Weapons/Sledgehammer/sledgehammer_impact_05}}
 {id=WavePlayer_24, properties {soundWaveAssetPtr=&Industrial_Distant_Impact03}}
 {id=WavePlayer_29, properties {soundWaveAssetPtr=&Industrial_Distant_Impact02}}
 {id=WavePlayer_30, properties {soundWaveAssetPtr=&Industrial_Distant_Impact04}}
 {id=WavePlayer_31, properties {soundWaveAssetPtr=&Industrial_Distant_Impact10}}
 {id=WavePlayer_32, properties {soundWaveAssetPtr=&sledgehammer_impact_01}}
 {id=WavePlayer_33, properties {soundWaveAssetPtr=&sledgehammer_impact_02}}
 {id=WavePlayer_34, properties {soundWaveAssetPtr=&sledgehammer_impact_04}}
]
properties {
 bOverrideAttenuation=true
 maxDistance=2097152
 volumeMultiplier=1
 attenuationOverrides {bAttenuate=false}
 concurrencyOverrides {resolutionRule=PreventNew}
 concurrencySet [&Player_Melee_Weapon_Conc]
}
