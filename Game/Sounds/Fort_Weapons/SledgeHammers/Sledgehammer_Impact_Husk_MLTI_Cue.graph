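# Sound cue for a sledgehammer hitting a Husk, limited by the player melee weapon concurrency set.
# A quality-level switch picks between two layered branches, the second with fewer random wave variations, and leaves its other two slots empty.
# Each branch mixes a close melee layer, a hammer impact plus briefly delayed Husk impact, with a distant industrial-impact layer.
# Close and distant layers have separate sound classes and attenuations, the distant one using low-pass filtering.
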SoundCue /Game/Sounds/Fort_Weapons/SledgeHammers/Sledgehammer_Impact_Husk_MLTI_Cue
root=QualityLevel_0
nodes [
 {id=Attenuation_2, children [Mixer_2], properties {attenuationSettings=&Fort_Weapon_Melee_Sml_Att}}
 {
  id=Attenuation_3
  children [Random_5]
  properties {
   attenuationSettings=/Game/Sounds/Attenuation/Fort_Weapon_Projectile_Distant_StW_Att
   attenuationOverrides {
    bAttenuateWithLPF=true
    distanceAlgorithm=Logarithmic
    falloffDistance=8000
    lPFRadiusMax=5000
    lPFRadiusMin=2000
    attenuationShapeExtents {x=5000}
   }
  }
 }
 {id=Attenuation_4, children [Mixer_4], properties {attenuationSettings=&Fort_Weapon_Melee_Sml_Att}}
 {
  id=Attenuation_5
  children [Modulator_3]
  properties {
   attenuationSettings=&Fort_Weapon_Projectile_Distant_StW_Att
   attenuationOverrides {
    bAttenuateWithLPF=true
    distanceAlgorithm=Logarithmic
    falloffDistance=8000
    lPFRadiusMax=5000
    lPFRadiusMin=2000
    attenuationShapeExtents {x=5000}
   }
  }
 }
 {id=Delay_1, children [Random_4], properties {delayMax=0.1, delayMin=0.1}}
 {id=Delay_2, children [Random_7], properties {delayMax=0.1, delayMin=0.1}}
 {id=Mixer_2, children [Modulator_1,Delay_1], properties {inputVolume [1,1]}}
 {id=Mixer_3, children [SoundClass_3,SoundClass_2], properties {inputVolume [1,1]}}
 {id=Mixer_4, children [Modulator_2,Delay_2], properties {inputVolume [1,1]}}
 {id=Mixer_5, children [SoundClass_5,SoundClass_4], properties {inputVolume [1,1]}}
 {id=Modulator_1, children [Random_3]}
 {id=Modulator_2, children [Random_6]}
 {id=Modulator_3, children [Random_8]}
 {id=QualityLevel_0, children [Mixer_3,Mixer_5,null,null]}
 {id=Random_3, children [WavePlayer_14, WavePlayer_15, WavePlayer_16, WavePlayer_17], properties {weights [1,1,1,1]}}
 {
  id=Random_4
  children [WavePlayer_18, WavePlayer_19, WavePlayer_20, WavePlayer_21, WavePlayer_22, WavePlayer_23]
  properties {weights [1,1,1,1,1,1]}
 }
 {id=Random_5, children [WavePlayer_26, WavePlayer_24, WavePlayer_27, WavePlayer_25], properties {weights [1,1,1,1]}}
 {id=Random_6, children [WavePlayer_28, WavePlayer_29, WavePlayer_30], properties {weights [1,1,1]}}
 {id=Random_7, children [WavePlayer_32, WavePlayer_33, WavePlayer_34, WavePlayer_36], properties {weights [1,1,1,1]}}
 {id=Random_8, children [WavePlayer_38, WavePlayer_41], properties {weights [1,1]}}
 {
  id=SoundClass_2
  children [Attenuation_3]
  properties {soundClassOverride=&Fort_WeaponRanged_Fire_Distant}
 }
 {id=SoundClass_3, children [Attenuation_2], properties {soundClassOverride=&Fort_WeaponMelee_PSM}}
 {
  id=SoundClass_4
  children [Attenuation_5]
  properties {soundClassOverride=&Fort_WeaponRanged_Fire_Distant}
 }
 {id=SoundClass_5, children [Attenuation_4], properties {soundClassOverride=&Fort_WeaponMelee_PSM}}
 {id=WavePlayer_14, properties {soundWaveAssetPtr=/Game/Sounds/Fort_Weapons/Sledgehammer/sledgehammer_impact_01}}
 {id=WavePlayer_15, properties {soundWaveAssetPtr=/Game/Sounds/Fort_Weapons/Sledgehammer/sledgehammer_impact_02}}
 {id=WavePlayer_16, properties {soundWaveAssetPtr=/Game/Sounds/Fort_Weapons/Sledgehammer/sledgehammer_impact_04}}
 {id=WavePlayer_17, properties {soundWaveAssetPtr=/Game/Sounds/Fort_Weapons/Sledgehammer/sledgehammer_impact_05}}
 {id=WavePlayer_18, properties {soundWaveAssetPtr=/Game/Sounds/Fort_Monster_Husk/SFX/husk_impact_mix_ST_09}}
 {id=WavePlayer_19, properties {soundWaveAssetPtr=/Game/Sounds/Fort_Monster_Husk/SFX/husk_impact_mix_ST_13}}
 {id=WavePlayer_20, properties {soundWaveAssetPtr=/Game/Sounds/Fort_Monster_Husk/SFX/husk_impact_mix_ST_01}}
 {id=WavePlayer_21, properties {soundWaveAssetPtr=/Game/Sounds/Fort_Monster_Husk/SFX/husk_impact_mix_ST_07}}
 {id=WavePlayer_22, properties {soundWaveAssetPtr=/Game/Sounds/Fort_Monster_Husk/SFX/husk_impact_mix_ST_10}}
 {id=WavePlayer_23, properties {soundWaveAssetPtr=/Game/Sounds/Fort_Monster_Husk/SFX/husk_impact_mix_ST_04}}
 {id=WavePlayer_24, properties {soundWaveAssetPtr=/Game/Sounds/Fort_Ambient_NonLoop/Industrial/Industrial_Distant_Impact03}}
 {id=WavePlayer_25, properties {soundWaveAssetPtr=/Game/Sounds/Fort_Ambient_NonLoop/Industrial/Industrial_Distant_Impact06}}
 {id=WavePlayer_26, properties {soundWaveAssetPtr=/Game/Sounds/Fort_Ambient_NonLoop/Industrial/Industrial_Distant_Impact02}}
 {id=WavePlayer_27, properties {soundWaveAssetPtr=/Game/Sounds/Fort_Ambient_NonLoop/Industrial/Industrial_Distant_Impact04}}
 {id=WavePlayer_28, properties {soundWaveAssetPtr=&sledgehammer_impact_01}}
 {id=WavePlayer_29, properties {soundWaveAssetPtr=&sledgehammer_impact_02}}
 {id=WavePlayer_30, properties {soundWaveAssetPtr=&sledgehammer_impact_04}}
 {id=WavePlayer_32, properties {soundWaveAssetPtr=&husk_impact_mix_ST_09}}
 {id=WavePlayer_33, properties {soundWaveAssetPtr=&husk_impact_mix_ST_13}}
 {id=WavePlayer_34, properties {soundWaveAssetPtr=&husk_impact_mix_ST_01}}
 {id=WavePlayer_36, properties {soundWaveAssetPtr=&husk_impact_mix_ST_10}}
 {id=WavePlayer_38, properties {soundWaveAssetPtr=&Industrial_Distant_Impact03}}
 {id=WavePlayer_41, properties {soundWaveAssetPtr=&Industrial_Distant_Impact04}}
]
properties {
 maxDistance=20000
 volumeMultiplier=0.5
 attenuationOverrides {bAttenuate=false}
 concurrencySet [&Player_Melee_Weapon_Conc]
}
