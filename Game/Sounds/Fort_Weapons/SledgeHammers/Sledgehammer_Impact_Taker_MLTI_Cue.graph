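# Sound cue for sledgehammer impacts on metal, layered by listener distance through close, medium and far attenuation crossfades.
# The close layer mixes a random sledgehammer hit with a slightly delayed default metal impact, both modulated.
# The medium layer plays a delayed, upward pitch-varied big metal hit; the far layer plays a delayed, quieter distant metal impact.
# Output uses the impacts sound class and player melee weapon concurrency.
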SoundCue /Game/Sounds/Fort_Weapons/SledgeHammers/Sledgehammer_Impact_Taker_MLTI_Cue
root=Mixer_5
nodes [
 {
  id=Attenuation_3
  children [Delay_4]
  properties {attenuationSettings=/Game/Sounds/Attenuation/Fort_Demolition_Crossfade_Global_Hit_Medium_Att, attenuationOverrides {falloffDistance=1600}}
 }
 {
  id=Attenuation_4
  children [Delay_5]
  properties {
   attenuationSettings=/Game/Sounds/Attenuation/Fort_Demolition_Crossfade_Global_Hit_Far_Att
   attenuationOverrides {falloffDistance=3000, attenuationShapeExtents {x=1000}}
  }
 }
 {
  id=Attenuation_5
  children [Mixer_3]
  properties {
   attenuationSettings=/Game/Sounds/Attenuation/Fort_Demolition_Crossfade_Global_Hit_Close_Att
   attenuationOverrides {falloffDistance=900, attenuationShapeExtents {x=100}}
  }
 }
 {id=Delay_3, children [Modulator_4], properties {delayMax=0.1, delayMin=0.1}}
 {id=Delay_4, children [Modulator_3], properties {delayMax=0.2, delayMin=0.1}}
 {id=Delay_5, children [Random_5], properties {delayMax=0.2, delayMin=0.1}}
 {id=Mixer_3, children [Modulator_5,Delay_3], properties {inputVolume [1.3,1]}}
 {id=Mixer_4, children [Attenuation_5, Attenuation_3], properties {inputVolume [1,1]}}
 {id=Mixer_5, children [Mixer_4, Attenuation_4], properties {inputVolume [1,0.7]}}
 {id=Modulator_3, children [Random_4], properties {pitchMax=1.8, pitchMin=1, volumeMax=1.3, volumeMin=1.3}}
 {id=Modulator_4, children [Random_6]}
 {id=Modulator_5, children [Random_7]}
 {id=Random_4, children [WavePlayer_17, WavePlayer_18, WavePlayer_19], properties {weights [1,1,1]}}
 {id=Random_5, children [WavePlayer_20, WavePlayer_21, WavePlayer_22], properties {weights [1,1,1]}}
 {
  id=Random_6
  children [WavePlayer_23, WavePlayer_24, WavePlayer_25, WavePlayer_26, WavePlayer_27, WavePlayer_28, WavePlayer_29]
  properties {weights [1,1,1,1,1,1,1]}
 }
 {id=Random_7, children [WavePlayer_30, WavePlayer_31, WavePlayer_32, WavePlayer_33], properties {weights [1,1,1,1]}}
 {id=WavePlayer_17, properties {soundWaveAssetPtr=/Game/Sounds/Fort_Impact_Sounds/Demolition/Metal/Metal_Big_01}}
 {id=WavePlayer_18, properties {soundWaveAssetPtr=/Game/Sounds/Fort_Impact_Sounds/Demolition/Metal/Metal_Big_03}}
 {id=WavePlayer_19, properties {soundWaveAssetPtr=/Game/Sounds/Fort_Impact_Sounds/Demolition/Metal/Metal_Big_04}}
 {id=WavePlayer_20, properties {soundWaveAssetPtr=/Game/Sounds/Fort_Impact_Sounds/Demolition/Defaults/Metal_Distant_Impact_01}}
 {id=WavePlayer_21, properties {soundWaveAssetPtr=/Game/Sounds/Fort_Impact_Sounds/Demolition/Defaults/Metal_Distant_Impact_03}}
 {id=WavePlayer_22, properties {soundWaveAssetPtr=/Game/Sounds/Fort_Impact_Sounds/Demolition/Defaults/Metal_Distant_Impact_04}}
 {id=WavePlayer_23, properties {soundWaveAssetPtr=/Game/Sounds/Fort_Impact_Sounds/Demolition/Defaults/default_metal_impact_new_combined_01}}
 {id=WavePlayer_24, properties {soundWaveAssetPtr=/Game/Sounds/Fort_Impact_Sounds/Demolition/Defaults/default_metal_impact_new_combined_02}}
 {id=WavePlayer_25, properties {soundWaveAssetPtr=/Game/Sounds/Fort_Impact_Sounds/Demolition/Defaults/default_metal_impact_new_combined_03}}
 {id=WavePlayer_26, properties {soundWaveAssetPtr=/Game/Sounds/Fort_Impact_Sounds/Demolition/Defaults/default_metal_impact_new_combined_04}}
 {id=WavePlayer_27, properties {soundWaveAssetPtr=/Game/Sounds/Fort_Impact_Sounds/Demolition/Defaults/default_metal_impact_new_combined_05}}
 {id=WavePlayer_28, properties {soundWaveAssetPtr=/Game/Sounds/Fort_Impact_Sounds/Demolition/Defaults/default_metal_impact_new_combined_06}}
 {id=WavePlayer_29, properties {soundWaveAssetPtr=/Game/Sounds/Fort_Impact_Sounds/Demolition/Defaults/default_metal_impact_new_combined_07}}
 {id=WavePlayer_30, properties {soundWaveAssetPtr=/Game/Sounds/Fort_Weapons/Sledgehammer/sledgehammer_impact_01}}
 {id=WavePlayer_31, properties {soundWaveAssetPtr=/Game/Sounds/Fort_Weapons/Sledgehammer/sledgehammer_impact_02}}
 {id=WavePlayer_32, properties {soundWaveAssetPtr=/Game/Sounds/Fort_Weapons/Sledgehammer/sledgehammer_impact_04}}
 {id=WavePlayer_33, properties {soundWaveAssetPtr=/Game/Sounds/Fort_Weapons/Sledgehammer/sledgehammer_impact_05}}
]
properties {
 attenuationSettings=&Fort_Weapon_Melee_Sml_Att
 maxDistance=8000
 soundClassObject=&Fort_Impacts
 volumeMultiplier=0.442
 concurrencyOverrides {resolutionRule=PreventNew}
 concurrencySet [&Player_Melee_Weapon_Conc]
}
